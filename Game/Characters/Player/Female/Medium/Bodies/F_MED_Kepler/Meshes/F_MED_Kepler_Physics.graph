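# Physics asset for the female medium Kepler player body mesh, covering only the left and right dynamic zipper bones.
# Each side has a small box body on the base bone and a larger, offset box body on the zipper_parent bone.
# One constraint per side joins the pair with tight, asymmetric swing limits and very little twist, restricting how far each zipper can move relative to its base.
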PhysicsAsset /Game/Characters/Player/Female/Medium/Bodies/F_MED_Kepler/Meshes/F_MED_Kepler_Physics
bodies {
 dyn_zipper_base_l [{shapeType=Box, extentX=0.5, extentY=0.5, extentZ=0.5}]
 dyn_zipper_base_r [{shapeType=Box, extentX=0.5, extentY=0.5, extentZ=0.5}]
 dyn_zipper_parent_l [{shapeType=Box, extentX=5.51, extentY=3.76, extentZ=1.25, center=<2.3851 0 0>}]
 dyn_zipper_parent_r [{shapeType=Box, extentX=5.51, extentY=3.76, extentZ=1.25, center=<2.3851 0 0>}]
}
constraints [
 {bone1Name=dyn_zipper_parent_l, bone2Name=dyn_zipper_base_l, swing1LimitDegrees=20, swing2LimitDegrees=10, twistLimitDegrees=3}
 {bone1Name=dyn_zipper_parent_r, bone2Name=dyn_zipper_base_r, swing1LimitDegrees=20, swing2LimitDegrees=10, twistLimitDegrees=3}
]
