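# Post-process style AnimBP for the F_MED_Kepler female body, adding secondary motion to back suspenders and zippers on the pose copied from the attached parent mesh.
# Weighted constraints anchor suspender and zipper base bones, spring AnimDynamics swings the suspenders, then RigidBody simulates the body's physics asset.
# Delayed jump begin/end events both set Jump_Impluse, the suspenders' external force, to the same upward vector, never resetting it.
# Two zipper AnimDynamics nodes are unconnected.
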AnimBlueprint /Game/Characters/Player/Female/Medium/Bodies/F_MED_Kepler/Meshes/F_MED_Kepler_AnimBP
target_skeleton=&/F_MED_Kepler_Skeleton
graphs [
 AnimGraph {nodes [
   Anim Dynamics - Bone: dyn_zipper_parent_l(Alpha=1, ExternalForce="0,0,0") {
    angularDampingOverride=0.9999
    bGravityOverrideInSimSpace=true
    bOverrideAngularDamping=true
    bOverrideLinearDamping=true
    linearDampingOverride=0.9999
    simulationSpace=RootRelative
    alphaBoolBlend {blendOption=Linear}
    boundBone {boneName=dyn_zipper_parent_l}
    componentAppliedLinearAccClamp=<3000 3000 3000>
    componentLinearAccScale=<0.25 0.25 0.25>
    componentLinearVelScale=<0.25 0.25 0.25>
    physicsBodyDefinitions [{
      sphereCollisionRadius=0
      boundBone {boneName=dyn_zipper_parent_l}
      boxExtents=<6 4 -1.5>
      constraintSetup {angularLimitsMax=<0 50 20>, angularLimitsMin=<0 20 -20>, angularTarget=<0 0 0>}
      localJointOffset=<2 0 0>
     }]
    relativeSpaceBone {boneName=root}
   }
   Anim Dynamics - Bone: dyn_zipper_parent_r(Alpha=1, ExternalForce="0,0,0") {
    angularDampingOverride=0.9999
    bGravityOverrideInSimSpace=true
    bOverrideAngularDamping=true
    bOverrideLinearDamping=true
    linearDampingOverride=0.9999
    simulationSpace=RootRelative
    alphaBoolBlend {blendOption=Linear}
    boundBone {boneName=dyn_zipper_parent_r}
    componentAppliedLinearAccClamp=<3000 3000 3000>
    componentLinearAccScale=<0.25 0.25 0.25>
    componentLinearVelScale=<0.25 0.25 0.25>
    physicsBodyDefinitions [{
      sphereCollisionRadius=0
      boundBone {boneName=dyn_zipper_parent_r}
      boxExtents=<6 4 -1>
      constraintSetup {angularLimitsMax=<0 20 20>, angularLimitsMin=<0 -2 -20>, angularTarget=<0 0 0>}
      localJointOffset=<2.5 0 0>
     }]
    relativeSpaceBone {boneName=root}
   }
   n1 = "Constraint - dyn_suspender_back_base_l (PPPPPPP)"!(
    ComponentPose=Local To Component(LocalPose=Copy Pose From Mesh() {bUseAttachedParent=true})
    Alpha=1
    ConstraintWeights_0=0.0025
    ConstraintWeights_1:float=0.1644
    ConstraintWeights_2:float=0.3293
    ConstraintWeights_3:float=0.2652
    ConstraintWeights_4:float=0.207
    ConstraintWeights_5:float=0.0022
    ConstraintWeights_6:float=0.0294
   ) -> (Pose) {
    alphaBoolBlend {blendOption=Linear}
    boneToModify {boneName=dyn_suspender_back_base_l}
    constraintSetup [
     {targetBone {boneName=clavicle_l}}
     {targetBone {boneName=spine_01}}
     {targetBone {boneName=spine_02}}
     {targetBone {boneName=spine_03}}
     {targetBone {boneName=spine_04}}
     {targetBone {boneName=spine_05}}
     {targetBone {boneName=pelvis}}
    ]
    constraintWeights [1,1,1,1,1,1,1]
   }
   n2 = Get Jump_Impluse() -> (Jump_Impluse:Vector)
   n3 = Anim Dynamics - Bone: dyn_suspender_back_parent_l(ComponentPose=@n1, Alpha=0.65, ExternalForce=@n2) {
    angularDampingOverride=0.9999
    bAngularSpring=true
    bGravityOverrideInSimSpace=true
    bLinearSpring=true
    bOverrideAngularDamping=true
    bOverrideLinearDamping=true
    gravityScale=0.75
    linearDampingOverride=0.9999
    alphaBoolBlend {blendOption=Linear}
    boundBone {boneName=dyn_suspender_back_parent_l}
    componentAppliedLinearAccClamp=<3000 3000 3000>
    componentLinearAccScale=<0.25 0.25 0.25>
    componentLinearVelScale=<0.25 0.25 0.25>
    gravityOverride=<1 1 1>
    physicsBodyDefinitions [{
      sphereCollisionRadius=0
      boundBone {boneName=dyn_suspender_back_parent_l}
      boxExtents=<6 4 -1>
      constraintSetup {angularLimitsMax=<0 40 0>, angularLimitsMin=<0 -10 0>, angularTarget=<0 0 0>}
      localJointOffset=<2.5 0 0>
     }]
    relativeSpaceBone {boneName=pelvis}
   }
   n4 = "Constraint - dyn_suspender_back_base_r (PPPPPPP)"!(
    ComponentPose=@n3
    Alpha=1
    ConstraintWeights_0=0.1673
    ConstraintWeights_1:float=0.332
    ConstraintWeights_2:float=0.262
    ConstraintWeights_3:float=0.202
    ConstraintWeights_4:float=0.0018
    ConstraintWeights_5:float=0.0021
    ConstraintWeights_6:float=0.0328
   ) -> (Pose) {
    alphaBoolBlend {blendOption=Linear}
    boneToModify {boneName=dyn_suspender_back_base_r}
    constraintSetup [
     {targetBone {boneName=spine_01}}
     {targetBone {boneName=spine_02}}
     {targetBone {boneName=spine_03}}
     {targetBone {boneName=spine_04}}
     {targetBone {boneName=spine_05}}
     {targetBone {boneName=clavicle_r}}
     {targetBone {boneName=pelvis}}
    ]
    constraintWeights [1,1,1,1,1,1,1]
   }
   n5 = Anim Dynamics - Bone: dyn_suspender_back_parent_r(ComponentPose=@n4, Alpha=0.65, ExternalForce=@n2) {
    angularDampingOverride=0.9999
    bAngularSpring=true
    bGravityOverrideInSimSpace=true
    bLinearSpring=true
    bOverrideAngularDamping=true
    bOverrideLinearDamping=true
    gravityScale=0.75
    linearDampingOverride=0.9999
    alphaBoolBlend {blendOption=Linear}
    boundBone {boneName=dyn_suspender_back_parent_r}
    componentAppliedLinearAccClamp=<3000 3000 3000>
    componentLinearAccScale=<0.25 0.25 0.25>
    componentLinearVelScale=<0.25 0.25 0.25>
    gravityOverride=<1 1 1>
    physicsBodyDefinitions [{
      sphereCollisionRadius=0
      boundBone {boneName=dyn_suspender_back_parent_r}
      boxExtents=<6 4 -1>
      constraintSetup {angularLimitsMax=<0 40 0>, angularLimitsMin=<0 -10 0>, angularTarget=<0 0 0>}
      localJointOffset=<2.5 0 0>
     }]
    relativeSpaceBone {boneName=pelvis}
   }
   n6 = "Constraint - dyn_zipper_base_r (PPPPPP)"!(
    ComponentPose=@n5
    Alpha=1
    ConstraintWeights_0=0
    ConstraintWeights_1:float=0.737
    ConstraintWeights_2:float=0.2611
    ConstraintWeights_3:float=0.0014
    ConstraintWeights_4:float=0.0001
    ConstraintWeights_5:float=0.0004
   ) -> (Pose) {
    alphaBoolBlend {blendOption=Linear}
    boneToModify {boneName=dyn_zipper_base_r}
    constraintSetup [
     {targetBone {boneName=calf_r}}
     {targetBone {boneName=thigh_r}}
     {targetBone {boneName=thigh_twist_01_r}}
     {targetBone {boneName=thigh_twist_01_l}}
     {targetBone {boneName=thigh_l}}
     {targetBone {boneName=pelvis}}
    ]
    constraintWeights [1,1,1,1,1,1]
   }
   n7 = "Constraint - dyn_zipper_base_l (PPP)"!(ComponentPose=@n6, Alpha=1, ConstraintWeights_0=0.0022, ConstraintWeights_1:float=0.2481, ConstraintWeights_2:float=0.7497) -> (Pose) {
    alphaBoolBlend {blendOption=Linear}
    boneToModify {boneName=dyn_zipper_base_l}
    constraintSetup [{targetBone {boneName=pelvis}}, {targetBone {boneName=thigh_twist_01_l}}, {targetBone {boneName=thigh_l}}]
    constraintWeights [1,1,1]
   }
   n8 = RigidBody(ComponentPose=@n7) {
    bClampLinearTranslationLimitToRefPose=true
    bOverrideWorldGravity=true
    overridePhysicsAsset=&/F_MED_Kepler_Physics
    simulationSpace=BaseBoneSpace
    alphaBoolBlend {blendOption=Linear}
    baseBoneRef {boneName=root}
    componentAppliedLinearAccClamp=<6000 6000 6000>
    componentLinearAccScale=<0.25 0.25 0.25>
    componentLinearVelScale=<0.25 0.25 0.25>
    overrideWorldGravity {z=-980}
   }
   Output Pose(Result=Component To Local(ComponentPose=@n8)) {layerGroup=None}
  ]}
 EventGraph {nodes [
   Event OnBeginJumping()
   > RetriggerableDelay(Duration=0.25)
   > Set Jump_Impluse(Jump_Impluse:Vector="0, 0,2000") -> (Output_Get:Vector)
   Event OnEndJumping()
   > RetriggerableDelay(Duration=0.375)
   > Set Jump_Impluse(Jump_Impluse="0, 0,2000")
   Event BlueprintUpdateAnimation()
  ]}
]
variables [Jump_Impluse {type=Vector}]
