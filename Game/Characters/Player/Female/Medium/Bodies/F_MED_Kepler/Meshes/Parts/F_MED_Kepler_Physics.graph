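# Physics asset for the F_MED_Kepler female medium player character, covering the head and the dynamic hair bones.
# Small capsules form left and right bang chains and a hair loop, each ending at its own base body.
# Four hair bodies attach in parallel to a sphere on the offset bone rather than chaining together.
# Constraints set tight swing and twist limits, while a large unconstrained head capsule provides the skull shape.
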PhysicsAsset /Game/Characters/Player/Female/Medium/Bodies/F_MED_Kepler/Meshes/Parts/F_MED_Kepler_Physics
bodies {
 dyn_bang_01_l [{shapeType=Capsule, length=1.5927, radius=0.8482, center=<1.5187 0 0.0002>, rotation=<rot 90 0 0>}]
 dyn_bang_01_r [{shapeType=Capsule, length=0.8455, radius=0.5, center=<0.6917 0.1521 -0.1787>, rotation=<rot 80.6206 179.9999 -180>}]
 dyn_bang_02_l [{shapeType=Capsule, length=1.1152, radius=0.5, center=<1.0154 0 0>, rotation=<rot 90 0 0>}]
 dyn_bang_base_l [{shapeType=Capsule, length=0, radius=0.5, rotation=<rot 90 0 0>}]
 dyn_bang_base_r [{shapeType=Capsule, length=0, radius=0.5}]
 dyn_bang_parent_l [{shapeType=Capsule, length=1.8375, radius=1.1487, center=<1.3017 0 0.0001>, rotation=<rot 90 0 0>}]
 dyn_bang_parent_r [{shapeType=Capsule, length=1.2858, radius=0.9615, center=<1.1732 0 0>, rotation=<rot 90 0 0>}]
 dyn_hair_01 [{shapeType=Capsule, length=0.6913, radius=0.3352, rotation=<rot 90 0 0>}]
 dyn_hair_02 [{shapeType=Capsule, length=0.755, radius=1.255, center=<1.2817 0 0.0001>, rotation=<rot 90 0 0>}]
 dyn_hair_03 [{shapeType=Capsule, length=0.88, radius=0.755, center=<1.0416 -0.0001 0>, rotation=<rot 90 0 0>}]
 dyn_hair_04 [{shapeType=Capsule, length=1.005, radius=0.755, center=<1.1491 0 0>, rotation=<rot 90 0 0>}]
 dyn_hair_loop_base [{shapeType=Capsule, length=0, radius=0.5}]
 dyn_hair_loop_parent [{shapeType=Capsule, length=1.6481, radius=0.6507, center=<-0.0628 0.0022 0.4218>, rotation=<rot 0 87.9841 -68.9311>}]
 dyn_hair_offset [{shapeType=Sphere, radius=0.5}]
 head [{shapeType=Capsule, length=1.8895, radius=6.5323, center=<2.9871 2.203 0.2148>, rotation=<rot 90 180 157.5177>}]
}
constraints [
 {bone1Name=dyn_bang_parent_l, bone2Name=dyn_bang_base_l, swing1LimitDegrees=5, swing2LimitDegrees=5, twistLimitDegrees=3}
 {bone1Name=dyn_bang_01_l, bone2Name=dyn_bang_parent_l, swing1LimitDegrees=8, swing2LimitDegrees=8, twistLimitDegrees=3}
 {bone1Name=dyn_bang_02_l, bone2Name=dyn_bang_01_l, swing1LimitDegrees=6, swing2LimitDegrees=6, twistLimitDegrees=3}
 {bone1Name=dyn_bang_parent_r, bone2Name=dyn_bang_base_r, swing1LimitDegrees=12, swing2LimitDegrees=12, twistLimitDegrees=3}
 {bone1Name=dyn_bang_01_r, bone2Name=dyn_bang_parent_r, swing1LimitDegrees=10, swing2LimitDegrees=10, twistLimitDegrees=3}
 {bone1Name=dyn_hair_loop_parent, bone2Name=dyn_hair_loop_base, swing1LimitDegrees=15, swing2LimitDegrees=15, twistLimitDegrees=3}
 {bone1Name=dyn_hair_01, bone2Name=dyn_hair_offset, swing1LimitDegrees=10, swing2LimitDegrees=10, twistLimitDegrees=3}
 {bone1Name=dyn_hair_02, bone2Name=dyn_hair_offset, swing1LimitDegrees=10, swing2LimitDegrees=10, twistLimitDegrees=3}
 {bone1Name=dyn_hair_03, bone2Name=dyn_hair_offset, swing1LimitDegrees=20, swing2LimitDegrees=20, twistLimitDegrees=3}
 {bone1Name=dyn_hair_04, bone2Name=dyn_hair_offset, swing1LimitDegrees=15, swing2LimitDegrees=15, twistLimitDegrees=3}
]
